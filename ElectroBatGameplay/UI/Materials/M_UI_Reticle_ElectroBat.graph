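# UI reticle material for the ElectroBat, drawing a center circle, mirrored brackets and a charge meter driven by ChargeCount and ProgressPercent.
# An MSDF bracket offset by SpreadOffset and an SDF circle outline receive anti-aliased edges and shadows.
# Radial arc masks fill the meter across three charge stages, with a sine-driven pulse on charged progress.
# Panned plasma textures in radial coordinates create the charged color effect, while SwingSwitch blends alternate coloring and opacity.
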Material /ElectroBatGameplay/UI/Materials/M_UI_Reticle_ElectroBat
graph {
 properties {bAutomaticallySetUsageInEditor=false, blendMode=BLEND_Translucent, materialDomain=MD_UI, usedShadingModels=Default Lit}
 nodes [
  n1 = MaterialFunctionCall!(UV, Segment Count=Constant() {r=2}, Segment Spacing=Constant() {r=0.6}, Fixed Spacing) {materialFunction=/FortUILibrary/MaterialFunctions/Basics/MF_UI_RadialSegments}
  n2 = TextureCoordinate()
  n3 = Abs(None=Add(A=ComponentMask(None=@n2) {r=true}) {constB=-0.5})
  n4 = ComponentMask(None=@n2) {g=true}
  n5 = Saturate(None=AppendVector(A=Add(A=@n3) {constB=0.5}, B=Multiply(A=@n4)))
  n6 = Constant() {r=0.08}
  n7 = Multiply(A=MaterialFunctionCall(
    UV=MaterialFunctionCall(UVs=@n5, "Rotation Angle (0-1)"=@n6) {materialFunction=&CustomRotator}.Rotated Values
    Segment Count=Constant() {r=7}
    Segment Spacing=Constant() {r=0.85}
   ) {materialFunction=&MF_UI_RadialSegments}.Segmented Lines Gradient)
  n8 = LinearInterpolate(Alpha=ScalarParameter() {parameterName=SpreadOffset, sliderMax=1}) {constB=0.2}
  n9 = Constant2Vector() {g=0.01, r=-0.07}
  n10 = MaterialFunctionCall!(UV, Offset=AppendVector(A=Add(A=@n8, B=@n9.R), B=@n9.G), Rotation, Scale=Multiply(A=Constant2Vector() {g=0.5, r=1}) {constB=1.5}, MirrorY) {materialFunction=/FortUILibrary/MaterialFunctions/Basics/MF_UI_Symmetry_2x1_DEPRECATED}
  n11 = Min(
   A=Max(A=Multiply(A=@n1.Segmented Triangles Gradient) {constB=3}, B=OneMinus(None=@n7))
   B=MaterialFunctionCall!(MSDF RGB=TextureSample(UVs=@n10.Result) {samplerType=SAMPLERTYPE_LinearColor, texture=/ElectroBatGameplay/UI/Textures/T_UII_MSDF_ElectroBat_Bracket}.RGB) {materialFunction=/FortUILibrary/MaterialFunctions/Shapes/Modifiers/MF_UI_MSDFtoSDF}.SDF
  )
  Composite!(SDF=NamedRerouteDeclaration(None=Multiply(A=@n11)) {name=SDFBracket, nodeColor=<rgba 1 1 1 1>}) {subgraphName=AA+Shadow}
  n12 = ScalarParameter() {defaultValue=0.15, parameterName=CircleSize}
  n13 = Multiply(A=MaterialFunctionCall!(
    DistanceField=MaterialFunctionCall!(UVs, Radius=@n12) {materialFunction=/FortUILibrary/MaterialFunctions/Shapes/SDFs/MF_SDF_Circle}.SDF
    Threshold
    Thickness=ScalarParameter() {defaultValue=0.065, parameterName=CircleThickness}
    Alignment
    Smoothing
   ) {materialFunction=/FortUILibrary/MaterialFunctions/Shapes/Modifiers/MF_UI_ShapeOutline}.SDFOutline) {constB=-10}
  Composite(SDF=@n13) {subgraphName=AA+Shadow}
  n14 = Constant4Vector() {constant=<rgba 1 1 1 0>}
  n15 = NamedRerouteUsage()
  n16 = NamedRerouteUsage()
  n17 = NamedRerouteUsage()
  n18 = Constant() {r=1}
  n19 = ScalarParameter() {parameterName=SwingSwitch}
  n20 = LinearInterpolate(
   B=LinearInterpolate(A=@n16, B=AppendVector(A=@n17, B=@n18), Alpha=@n19)
   Alpha=LinearInterpolate(A=NamedRerouteUsage(), B=NamedRerouteUsage(), Alpha=NamedRerouteUsage())
  )
  n21 = NamedRerouteUsage()
  n22 = LinearInterpolate(A=Constant4Vector() {constant=<rgba 0 0.5395 0.7157 1>}.RGBA, B=Constant4Vector() {constant=<rgba 1 1 1 0>}.RGBA, Alpha=@n19)
  n23 = NamedRerouteUsage()
  n24 = LinearInterpolate(B=Add(A=LinearInterpolate(B=@n16, Alpha=@n21), B=LinearInterpolate(B=@n22, Alpha=@n23)), Alpha=NamedRerouteUsage())
  Max(
   A=LinearInterpolate(B=Constant4Vector() {constant=<rgba 1 1 1 0>}.RGBA, Alpha=NamedRerouteUsage())
   B=Max(A=LinearInterpolate(B=@n14.RGBA, Alpha=@n15), B=Add(A=@n20, B=@n24))
  )
  n25 = NamedRerouteUsage()
  n26 = Constant() {desc=starting value, r=0.84}
  n27 = Constant() {desc=Ending Value, r=0.12}
  n28 = ScalarParameter() {defaultValue=0.5, parameterName=ProgressPercent}
  n29 = If(A=ScalarParameter() {parameterName=ChargeCount}, "A > B"=@n28, "A == B"=Constant() {r=1}, "A < B"=@n28) {constB=2}
  n30 = LinearInterpolate(
   A=@n26
   B=@n27
   Alpha=If(
    A=ScalarParameter() {parameterName=ChargeCount}
    "A > B"=LinearInterpolate(Alpha=@n29) {constA=0.6, desc=C3}
    "A == B"=LinearInterpolate(Alpha=@n29) {constA=0.22, constB=0.6, desc=C2}
    "A < B"=LinearInterpolate(Alpha=@n29) {constA=0.025, constB=0.2, desc=C1}
   )
  )
  n31 = NamedRerouteDeclaration(None=SmoothStep(Value=MaterialFunctionCall!(UV, "Arc Start Angle (0-1)", "Arc End Angle (0-1)", "Fill Ratio (0-1)"=@n30, Fill Clockwise) {materialFunction=/FortUILibrary/MaterialFunctions/Basics/MF_UI_RadialArc}.Centered Fill) {constMax=0.005}) {name=ProgressDelta, nodeColor=<rgba 1 1 1 1>}
  n32 = Multiply(A=NamedRerouteUsage(), B=@n31)
  n33 = NamedRerouteDeclaration(None=Subtract(A=@n25, B=NamedRerouteDeclaration(None=@n32) {name=ProgressMask, nodeColor=<rgba 1 1 1 1>})) {name=AABar, nodeColor=<rgba 1 1 1 1>}
  n34 = Multiply(A=@n31)
  n35 = ScalarParameter() {parameterName=ChargeCount}
  n36 = Constant() {r=0.2}
  n37 = If(
   A=@n35
   "A > B"=If(A=@n35, "A > B"=Constant() {r=1}, "A == B"=Constant() {r=0.6}, "A < B"=@n36) {constB=1}
   "A == B"=@n36
   "A < B"=Constant()
  )
  n38 = SmoothStep(Value=MaterialFunctionCall("Fill Ratio (0-1)"=LinearInterpolate(A=@n26, B=@n27, Alpha=@n37) {constAlpha=0}) {materialFunction=&MF_UI_RadialArc}.Centered Fill) {constMax=0.005}
  n39 = Sine(None=Time()) {period=0.5}
  n40 = LinearInterpolate(Alpha=Saturate(None=ConstantBiasScale(None=@n39))) {constA=0.4, constB=0.1}
  n41 = NamedRerouteDeclaration(None=Multiply(A=@n38, B=@n40)) {name=ProgressCharged, nodeColor=<rgba 1 1 1 1>}
  n42 = NamedRerouteDeclaration(None=Multiply(A=@n25, B=Add(A=@n34, B=@n41))) {name=AAProgress, nodeColor=<rgba 1 1 1 1>}
  n43 = NamedRerouteUsage()
  n44 = Multiply(A=SmoothStep(Min=@n40, Value=Multiply(A=@n43, B=@n38))) {constB=3}
  n45 = NamedRerouteUsage()
  n46 = NamedRerouteDeclaration(None=Saturate(None=Multiply(A=@n44, B=@n45))) {name=ChargedPulse, nodeColor=<rgba 1 1 1 1>}
  n47 = Constant() {r=1}
  n48 = NamedRerouteDeclaration(None=If(A=ScalarParameter() {parameterName=ChargeCount}, "A > B"=@n47, "A == B"=@n47, "A < B"=Constant()) {constB=2}) {name=ChargeSwitch, nodeColor=<rgba 1 1 1 1>}
  n49 = Multiply(A=NamedRerouteUsage(), B=NamedRerouteUsage())
  n50 = Multiply(A=NamedRerouteUsage()) {constB=0.25}
  n51 = ScalarParameter() {parameterName=SwingSwitch}
  n52 = Max(
   A=Max(A=Multiply(A=@n33) {constB=0.2}, B=Multiply(A=@n42))
   B=LinearInterpolate(A=LinearInterpolate(B=@n46, Alpha=@n48) {constB=0}, B=Add(A=@n49, B=@n50), Alpha=Multiply(A=@n51) {constB=1.25}) {constAlpha=1}
  ) {constB=0}
  MaterialExpressionReroute_3 = Reroute(None=@n13)
  n53 = Constant() {r=0.5}
  n54 = MaterialFunctionCall(DistanceField=@MaterialExpressionReroute_3.SDF, Threshold=@n53) {materialFunction=&MF_UI_AntiAliasing}
  MaterialExpressionReroute_4 = Reroute(None=@n54.Outer)
  n55 = NamedRerouteDeclaration(None=@MaterialExpressionReroute_4.AA'd Art) {name=AACircle, nodeColor=<rgba 1 1 1 1>}
  n56 = MaterialFunctionCall(DistanceField=@MaterialExpressionReroute_3.SDF, Threshold=Add(A=@n53) {constB=-0.15}) {materialFunction=&MF_UI_AntiAliasing}
  MaterialExpressionReroute_5 = Reroute(None=@n56.Outer)
  n57 = NamedRerouteUsage()
  Multiply(
   A=ScalarParameter() {defaultValue=0.85, parameterName=MasterOpacity}
   B=LinearInterpolate(A=Max(A=@n55, B=@n52), B=Max(A=@MaterialExpressionReroute_5.AA'd Shadow, B=@n57)) {constAlpha=0.4}
  )
  NamedRerouteDeclaration(None=@MaterialExpressionReroute_4.AA'd Art) {name=AABracket, nodeColor=<rgba 1 1 1 1>}
  n58 = Constant4Vector() {constant=<rgba 0 0.8723 1 1>}
  n59 = TextureCoordinate()
  n60 = ComponentMask(None=Rotator(Coordinate=Subtract(A=@n59) {constB=0.5}, Time=Constant() {r=0.25}) {speed=3.1416}) {r=true}
  n61 = Multiply(
   A=MaterialFunctionCall!(Vector or UVs) {materialFunction=/FortUILibrary/MaterialFunctions/Basics/MF_UI_CheaperVectorToRadial}.Radial Coordinates
   B=Constant2Vector() {g=0.5, r=5}
  )
  n62 = Time()
  n63 = Multiply(A=MaterialFunctionCall(Radius=Constant() {r=3}, Density=Constant() {r=4}) {materialFunction=&RadialGradientExponential}.RadialGradientExponential) {constB=-0.2}
  n64 = TextureSample(UVs=MaterialFunctionCall!(Value=Panner(Coordinate=@n61, Time=@n62, Speed=@n63) {speedX=0.6, speedY=-0.4}, Scale=Constant() {r=0.6}, Height) {materialFunction=/FortUILibrary/MaterialFunctions/Basics/MF_UI_PingPong}.Result) {samplerType=SAMPLERTYPE_LinearGrayscale, texture=/Game/Effects/Fort_Effects/Emotes/UltraLight/Texture/T_PlasmaCoils}
  n65 = Power(Base=NamedRerouteDeclaration(None=Multiply(A=@n64.RGB) {constB=3}) {name=Distort, nodeColor=<rgba 1 1 1 1>}) {constExponent=4.4}
  n66 = Power(Base=TextureSample(UVs=Panner(Coordinate=@n61, Time=@n62) {speedX=1, speedY=0.2}) {samplerType=SAMPLERTYPE_LinearGrayscale, texture=&T_PlasmaCoils}.RGB) {constExponent=0.2}
  n67 = MaterialFunctionCall(
   A=Constant4Vector() {constant=<rgba 0 0.3579 0.571 1>}.RGBA
   B=@n58.RGBA
   C=LinearInterpolate(A=@n58.RGBA, B=Constant4Vector() {constant=<rgba 0.3073 1 1 1>}.RGBA, Alpha=Subtract(A=@n60) {constB=0})
   Alpha=Power(Base=Multiply(A=@n65, B=@n66)) {constExponent=0.2}
  ) {materialFunction=&3ColorBlend}
  NamedRerouteDeclaration(None=AppendVector(A=Multiply(A=@n67.Result) {constB=0.75}, B=Constant() {r=1})) {name=ChargedFX, nodeColor=<rgba 0 1 0.8125 1>}
  NamedRerouteDeclaration(None=@MaterialExpressionReroute_5.AA'd Shadow) {name=AABracketShadow, nodeColor=<rgba 1 1 1 1>}
  PinBase() {reroutePins [SDF {expression=&self:MaterialGraph_0.MaterialGraphNode_Composite_1.AA+Shadow.MaterialExpressionReroute_3}]}
  n68 = Constant() {r=0.5}
  PinBase!(
   AA'd Art=MaterialFunctionCall(DistanceField=@MaterialExpressionReroute_3.SDF, Threshold=@n68) {materialFunction=&MF_UI_AntiAliasing}.Outer
   AA'd Shadow=MaterialFunctionCall(DistanceField=@MaterialExpressionReroute_3.SDF, Threshold=Add(A=@n68) {constB=-0.15}) {materialFunction=&MF_UI_AntiAliasing}.Outer
  ) {reroutePins [
    AA'd Art {expression=&self:MaterialGraph_0.MaterialGraphNode_Composite_1.AA+Shadow.MaterialExpressionReroute_4}
    AA'd Shadow {expression=&self:MaterialGraph_0.MaterialGraphNode_Composite_1.AA+Shadow.MaterialExpressionReroute_5}
   ]}
  PinBase!() {reroutePins [SDF {expression=&self:MaterialExpressionReroute_3}]}
  PinBase!(AA'd Art=@n54.Outer, AA'd Shadow=@n56.Outer) {reroutePins [
    AA'd Art {expression=&self:MaterialExpressionReroute_4}
    AA'd Shadow {expression=&self:MaterialExpressionReroute_5}
   ]}
 ]
}
