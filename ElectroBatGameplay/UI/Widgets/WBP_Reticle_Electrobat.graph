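# Electro Bat weapon reticle widget that shows staged charge progress and pulses on charged swings by driving a dynamic material.
# On view target change, it caches the pawn and equipped bat, then listens asynchronously for Charging and ChargedSwing gameplay tags.
# Charging starts a delay-timer and progress-animation loop, incrementing ChargeCount up to the weapon's stage limit, using timings from weapon scalable floats.
# Tag removal stops the matching animation and resets its material parameters.
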WidgetBlueprint /ElectroBatGameplay/UI/Widgets/WBP_Reticle_Electrobat
animations [
 Anim_ElectroBat_Charging {
  binding_count=1
  length_seconds=1
  object=&self:Anim_ElectroBat_Charging
  sequence {
   bindings [Image_Reticle {
     id=#1
     possessed_class=Image
     tracks [{
       type=MovieSceneWidgetMaterialTrack
       sections [{
         type=MovieSceneParameterSection
         end_frame=30
         start_frame=0
         channels [ProgressPercent {type=MovieSceneScriptingFloatChannel, keys [[0,0],[30,1]]}]
        }]
      }]
    }]
   display_rate {numerator=30}
  }
 }
 Anim_ElectroBat_Attack {
  binding_count=1
  length_seconds=0.3334
  object=&self:Anim_ElectroBat_Attack
  sequence {
   bindings [Image_Reticle {
     id=#2
     possessed_class=Image
     tracks [{
       type=MovieSceneWidgetMaterialTrack
       sections [{
         type=MovieSceneParameterSection
         end_frame=10
         start_frame=0
         channels [
          SpreadOffset {type=MovieSceneScriptingFloatChannel, keys [[0,0],[1,1],[6,0]]}
          SwingSwitch {type=MovieSceneScriptingFloatChannel, keys [[0,0],[1,1],[2,1],[10,0]]}
          MasterOpacity {type=MovieSceneScriptingFloatChannel, keys [[0,0.85],[1,1],[8,0.85]]}
         ]
        }]
      }]
    }]
   display_rate {numerator=30}
  }
 }
]
graphs [EventGraph {nodes [
   TimerDelay() {is=title K2Node_CustomEvent}
   n1 = Get As B Athena Electro Bat() -> (As B Athena Electro Bat:B Athena Electro Bat Object Reference)
   > n2 = Set Timer by Event(
    Delegate=Create Event()
    Time=GetCachedScalableFloatValue!(Input:"Fort Cached Scalable Float Structure (by ref)"=Get Row Pre Charge Timer!(self:B Athena Electro Bat Object Reference=@n1) -> (Row Pre Charge Timer:Fort Cached Scalable Float Structure)) -> (ReturnValue:float)
    bLooping=false
   )
   > Set ChargeTimerHandler(ChargeTimerHandler:Timer Handle Structure=@n2) -> (Output_Get:Timer Handle Structure)
   ChargeTimer() {is=title K2Node_CustomEvent}
   n3 = Get As B Athena Electro Bat()
   > n4 = "Play Animation with Finished event (legacy)"!(
    execute:Exec
    Widget:User Widget Object Reference=Self-Reference()
    InAnimation:Widget Animation Object Reference=Get Anim_ElectroBat_Charging() -> (Anim_ElectroBat_Charging:Widget Animation Object Reference)
    StartAtTime:float=0
    NumLoopsToPlay:int=1
    PlayMode:EUMGSequencePlayMode Enum
    PlaybackSpeed:float=GetCachedScalableFloatValue(Input=Get Row Charge Time Per Charge!(self:B Athena Electro Bat Object Reference=@n3) -> (Row Charge Time Per Charge:Fort Cached Scalable Float Structure))
   ) -> (then:Exec, Result:UMGSequence Player Object Reference, Finished:Exec) {is=title K2Node_PlayAnimation}
   n5 = Get ChargingTag() -> (ChargingTag:Gameplay Tag Structure)
   n6 = Get ViewTargetPlayerPawn() -> (ViewTargetPlayerPawn:Player Pawn Athena Object Reference)
   Branch(
    execute=@n4.Finished
    Condition=HasMatchingGameplayTag!(self:Gameplay Tag Asset Interface Interface=@n6, TagToCheck:Gameplay Tag Structure=@n5) -> (ReturnValue:bool)
   )
   > n7 = Increment Int(Value=Get ChargeCount() -> (ChargeCount:int=0))
   n8 = Get M_Reticle() -> (M_Reticle:Material Instance Dynamic Object Reference)
   > SetScalarParameterValue(self=@n8, ParameterName=ChargeCount, Value="To Float (Integer)"(InInt=@n7))
   > SetScalarParameterValue(self=@n8, ParameterName=ProgressPercent, Value=0)
   n9 = Get Row Charge Stages!(self:B Athena Electro Bat Object Reference=Get As B Athena Electro Bat()) -> (Row Charge Stages:Fort Cached Scalable Float Structure)
   > n10 = Branch(Condition="float >= float"(A:int=@n7, B=GetCachedScalableFloatValue(Input=@n9)))
   > Clear and Invalidate Timer by Handle(Handle=Get ChargeTimerHandler() -> (ChargeTimerHandler:Timer Handle Structure))
   TimerDelay(execute=@n10.else) {is=title K2Node_CallFunction}
   n11 = OnViewTargetChanged!() -> (
    OutputDelegate
    then
    PlayerController:Fort Player Controller Object Reference
    OldViewTarget:Actor Object Reference
    NewViewTarget:Actor Object Reference
   ) {is=title K2Node_CustomEvent}
   > n12 = Cast To PlayerPawn_Athena(Object=@n11.NewViewTarget)
   > Set ViewTargetPlayerPawn(ViewTargetPlayerPawn:Player Pawn Athena Object Reference=@n12) -> (Output_Get:Player Pawn Athena Object Reference)
   n13 = GetEquippedWeapon!(self:Fort Weapon Reticle Extension Widget Base Object Reference) -> (ReturnValue:Fort Weapon Object Reference)
   > n14 = Set EquippedWeapon(EquippedWeapon:Fort Weapon Object Reference=@n13) -> (Output_Get:Fort Weapon Object Reference)
   > Is Valid(InputObject=@n14)
   > n15 = Cast To B_Athena_ElectroBat!(execute, Object=@n13) -> (then, CastFailed, AsB Athena Electro Bat:B Athena Electro Bat Object Reference) {is=title K2Node_DynamicCast}
   > Set As B Athena Electro Bat(As B Athena Electro Bat:B Athena Electro Bat Object Reference=@n15) -> (Output_Get:B Athena Electro Bat Object Reference)
   > n16 = WaitGameplayTagAddToActor!(execute:Exec, TargetActor:Actor Object Reference=@n6, Tag:Gameplay Tag Structure=@n5, OnlyTriggerOnce:bool=false) -> (then:Exec, AsyncTaskProxy:Ability Async Wait Gameplay Tag Added Object Reference, Added:Exec) {is=title K2Node_AsyncAction}
   > n17 = WaitGameplayTagRemoveFromActor!(execute:Exec, TargetActor:Actor Object Reference=@n6, Tag:Gameplay Tag Structure=@n5, OnlyTriggerOnce:bool=false) -> (then:Exec, AsyncTaskProxy:Ability Async Wait Gameplay Tag Removed Object Reference, Removed:Exec) {is=title K2Node_AsyncAction}
   n18 = Get ChargedSwingTag() -> (ChargedSwingTag:Gameplay Tag Structure)
   > n19 = WaitGameplayTagAddToActor(TargetActor=@n6, Tag=@n18, OnlyTriggerOnce=false)
   > n20 = WaitGameplayTagRemoveFromActor(TargetActor=@n6, Tag=@n18, OnlyTriggerOnce=false)
   n21 = Get Anim_ElectroBat_Attack() -> (Anim_ElectroBat_Attack:Widget Animation Object Reference)
   StopAnimation(execute=@n20.Removed, InAnimation=@n21)
   n22 = Get M_Reticle()
   > SetScalarParameterValue(self=@n22, ParameterName=SpreadOffset, Value=0)
   PlayAnimation(execute=@n19.Added, InAnimation=@n21)
   Clear and Invalidate Timer by Handle(execute=@n17.Removed, Handle=Get ChargeTimerHandler())
   > SetScalarParameterValue(self=@n22, ParameterName=ChargeCount, Value=-1)
   > SetScalarParameterValue(self=@n22, ParameterName=ProgressPercent, Value=0)
   > StopAnimation(InAnimation=Get Anim_ElectroBat_Charging())
   n23 = Set ChargeCount(execute=@n16.Added, ChargeCount:int=0) -> (Output_Get:int=0)
   > SetScalarParameterValue(self=@n22, ParameterName=ChargeCount, Value="To Float (Integer)"(InInt=@n23))
   > TimerDelay()
   Event Construct()
   n24 = GetOwningPlayer()
   > n25 = Cast To FortPlayerController(Object=@n24)
   n26 = Create Event()
   > Bind Event to On View Target Changed(self:Fort Player Controller Object Reference=@n25, Delegate=@n26) {is=title K2Node_AddDelegate}
   > OnViewTargetChanged!(
    execute
    self
    PlayerController:Fort Player Controller Object Reference
    OldViewTarget:Actor Object Reference
    NewViewTarget:Actor Object Reference=GetViewTarget!(self:Controller Object Reference=@n24) -> (ReturnValue:Actor Object Reference)
   ) -> (then) {is=title K2Node_CallFunction}
   Event Destruct()
   > n27 = Cast To FortPlayerController(Object=@n24)
   > Unbind Event from On View Target Changed(self:Fort Player Controller Object Reference=@n27, Delegate=@n26) {is=title K2Node_RemoveDelegate}
   Event OnInitialized()
   > n28 = GetDynamicMaterial(self=Get Image_Reticle() -> (Image_Reticle:Image Object Reference))
   > Set M_Reticle(M_Reticle:Material Instance Dynamic Object Reference=@n28) -> (Output_Get:Material Instance Dynamic Object Reference)
  ]}]
variables [
 AbilitySystemComponentVar {type=AbilitySystemComponent, default=null}
 As B Athena Electro Bat {type=B_Athena_ElectroBat_C, default=null}
 Charge {type=GameplayTag, default {tagName=None}}
 ChargeCount {type=integer}
 ChargedSwingTag {type=GameplayTag, default {tagName=Gameplay.Status.ElectroBat.ChargedSwing.Started}}
 ChargeTimerHandler {type=TimerHandle}
 ChargingTag {type=GameplayTag, default {tagName=Gameplay.Status.ElectroBat.Charging}}
 CurrentChargeStage {
  type=GameplayAttribute
  default {
   attribute=/Script/ElectroBatGameplayRuntime.ElectroBatAttributeSet:CurrentChargeStage
   attributeName=CurrentChargeStage
   attributeOwner=&ElectroBatAttributeSet
  }
 }
 EquippedWeapon {type=FortWeapon, default=null}
 Image Reticle {type=Image, default=null}
 M_Reticle {type=MaterialInstanceDynamic}
 Row Charge Stages {type=FortCachedScalableFloat, default {scalableFloat {registryType=None, value=0, curve {curveTable=null, rowName=None}}}}
 Var_Weapon {type=FortMantisWeaponComponent, default=null}
 ViewTargetPlayerPawn {type=PlayerPawn_Athena_C}
]
widgets {
 info {
  inheritedWidgetCount=1
  parentClass=/Script/FortniteUI.FortWeaponReticleExtensionWidgetBase
  rootWidgetClass=&Image
  widgetCount=2
 }
 widgets [
  Image_Reticle {
   widgetClassPath=&Image
   bIsVariable=true
   widget=Image_Reticle
   properties {brush {
     bIsDynamicallyLoaded=false
     drawAs=Image
     imageType=FullColor
     mirroring=NoMirror
     resourceName=None
     resourceObject=/ElectroBatGameplay/UI/Materials/M_UI_Reticle_ElectroBat
     tiling=NoTile
     imageSize=<125 125>
     margin=<ltrb 0 0 0 0>
     outlineSettings {
      bUseBrushTransparency=false
      roundingType=HalfHeightRadius
      width=0
      color {colorUseRule=UseColor_Specified, specifiedColor=<rgba 0 0 0 0>}
      cornerRadii=<0 0 0 0>
     }
     tintColor {colorUseRule=UseColor_Specified, specifiedColor=<rgba 1 1 1 1>}
     uVRegion {bIsValid=false, max=<0 0>, min=<0 0>}
    }}
  }
  TutorialHighlight {widgetClassPath=&TutorialHighlight}
 ]
}
